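# Physics asset for the AsterOrder mic stand backpack cosmetic, defining bodies and joints for a handle with a dangling chain and end tag.
# A long capsule covers the handle, while zero-length capsules act as spheres along the chain and tag.
# Angular locks tie the chain base to the handle and the handle to its offset bone.
# Chain joints carry narrow twist limit values and swing values that widen mid-chain, then tighten toward the tag.
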PhysicsAsset /SparksCosmetics/Instrument/MicStand/AsterOrder/Meshes/AsterOrder_MicStand_Backpack_Physics
bodies {
 base_chain [{shapeType=Capsule, length=0, radius=0.5, rotation=<rot 90 0 0>}]
 dyn_chain_01 [{shapeType=Capsule, length=0, radius=0.75, center=<0.15 0 0>, rotation=<rot 90 0 0>}]
 dyn_chain_02 [{shapeType=Capsule, length=0, radius=0.75, center=<0.5 0 0>, rotation=<rot 90 0 0>}]
 dyn_chain_03 [{shapeType=Capsule, length=0, radius=0.75, center=<0.5 0 0>, rotation=<rot 90 0 0>}]
 dyn_chain_04 [{shapeType=Capsule, length=0, radius=0.75, center=<0.5 0 0>, rotation=<rot 90 0 0>}]
 dyn_chain_05 [{shapeType=Capsule, length=0, radius=0.75, center=<0.5 0 0>, rotation=<rot 90 0 0>}]
 dyn_chain_06 [{shapeType=Capsule, length=0, radius=0.75, center=<0.5 0 0>, rotation=<rot 90 0 0>}]
 dyn_chain_07 [{shapeType=Capsule, length=0, radius=0.75, center=<0.5 0 0>, rotation=<rot 90 0 0>}]
 dyn_chain_08 [{shapeType=Capsule, length=0, radius=0.75, center=<0.5 0 0>, rotation=<rot 90 0 0>}]
 dyn_tag [{shapeType=Capsule, length=0, radius=1.75, center=<1.3 0 0>, rotation=<rot 90 0 0>}]
 handle [{shapeType=Capsule, length=40, radius=2}]
 offset [{shapeType=Capsule, length=0.505, radius=0.505, rotation=<rot 0 -90 0>}]
 root [{shapeType=Capsule, length=0.505, radius=0.505, rotation=<rot 0 -90 0>}]
}
constraints [
 {bone1Name=offset, bone2Name=root, swing1LimitDegrees=45, swing2LimitDegrees=45, twistLimitDegrees=45}
 {
  bone1Name=handle
  bone2Name=offset
  swing1LimitDegrees=45
  swing1Motion=Locked
  swing2LimitDegrees=45
  swing2Motion=Locked
  twistLimitDegrees=45
  twistMotion=Locked
 }
 {
  bone1Name=base_chain
  bone2Name=handle
  swing1LimitDegrees=45
  swing1Motion=Locked
  swing2LimitDegrees=45
  swing2Motion=Locked
  twistLimitDegrees=45
  twistMotion=Locked
 }
 {bone1Name=dyn_chain_01, bone2Name=base_chain, swing1LimitDegrees=45, swing2LimitDegrees=45, twistLimitDegrees=2}
 {bone1Name=dyn_chain_02, bone2Name=dyn_chain_01, swing1LimitDegrees=45, swing2LimitDegrees=45, twistLimitDegrees=2}
 {bone1Name=dyn_chain_03, bone2Name=dyn_chain_02, swing1LimitDegrees=50, swing2LimitDegrees=50, twistLimitDegrees=2}
 {bone1Name=dyn_chain_04, bone2Name=dyn_chain_03, swing1LimitDegrees=55, swing2LimitDegrees=55, twistLimitDegrees=2}
 {bone1Name=dyn_chain_05, bone2Name=dyn_chain_04, swing1LimitDegrees=60, swing2LimitDegrees=60, twistLimitDegrees=2}
 {bone1Name=dyn_chain_06, bone2Name=dyn_chain_05, swing1LimitDegrees=60, swing2LimitDegrees=60, twistLimitDegrees=2}
 {bone1Name=dyn_chain_07, bone2Name=dyn_chain_06, swing1LimitDegrees=60, swing2LimitDegrees=60, twistLimitDegrees=2}
 {bone1Name=dyn_chain_08, bone2Name=dyn_chain_07, swing1LimitDegrees=45, swing2LimitDegrees=45, twistLimitDegrees=2}
 {bone1Name=dyn_tag, bone2Name=dyn_chain_08, swing1LimitDegrees=45, swing2LimitDegrees=45, twistLimitDegrees=2}
]
